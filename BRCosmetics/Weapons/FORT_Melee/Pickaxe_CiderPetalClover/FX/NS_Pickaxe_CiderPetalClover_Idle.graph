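# Idle Niagara effect for the CiderPetalClover pickaxe cosmetic, drawing persistent glow-colored spike meshes.
# Two burst emitters, Spike and Spike001, each use three mesh renderers, with FN_Mesh_Attributes writing extra transforms, camera offsets and material parameters for layering.
# User.ColorProgress blends colors from blue to warm gold-orange, and FN_FrontEndFloatModifier slightly varies the blue tint between gameplay and front end.
# Emitter-level palette colors follow the same blend, though no module shown here reads them.
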NiagaraSystem /BRCosmetics/Weapons/FORT_Melee/Pickaxe_CiderPetalClover/FX/NS_Pickaxe_CiderPetalClover_Idle
dependencies {usedDataInterfaces [], usedDynamicInputs [], usedModules [&SystemState]}
emitters {
 Spike {
  emitterUpdateScript [
   &EmitterState(
    Inactive Response=::NewEnumerator1
    Life Cycle Mode=::NewEnumerator1
    Loop Behavior=::NewEnumerator1
    Loop Duration Mode=::NewEnumerator1
    UseLoopDelay=false
   )
   &SpawnBurst_Instantaneous(Use Spawn Probability=false, UseLoopCountLimit=false)
   /SetVariables(
    Emitter.BaseBlueColor:LinearColor=&LerpLinearColors(StartColor=<rgba 0.767 1.234 5 1>, EndColor=<rgba 3 2.2031 1.4063 1>, LerpFactor=$User.ColorProgress)
    Emitter.BaseBlueColor02:LinearColor=&LerpLinearColors(StartColor=<rgba 0 1.8422 3 1>, EndColor=<rgba 3 2.2031 1.4063 1>, LerpFactor=$User.ColorProgress)
   ) {isSetParameters=true, module=SetVariables_#1}
   /SetVariables(
    Emitter.HighLightColor:LinearColor=&LerpLinearColors(StartColor=<rgba 0.1742 0.387 0.5 1>, EndColor=<rgba 1 0.7926 0.5853 1>, LerpFactor=$User.ColorProgress)
    Emitter.HighLightColor02:LinearColor=&LerpLinearColors(StartColor=<rgba 0 1.3561 5 1>, EndColor=<rgba 1 0.7926 0.5853 1>, LerpFactor=$User.ColorProgress)
   ) {isSetParameters=true, module=SetVariables_#2}
   /SetVariables(
    Emitter.BottomColorDark:LinearColor=&LerpLinearColors(StartColor=<rgba 0 0 0.05 1>, EndColor=<rgba 0.05 0.0417 0 1>, LerpFactor=$User.ColorProgress)
    Emitter.BottomColorDark02:LinearColor=&LerpLinearColors(StartColor=<rgba 0 0 0.05 1>, EndColor=<rgba 0.05 0.025 0 1>, LerpFactor=$User.ColorProgress)
   ) {isSetParameters=true, module=SetVariables_#3}
   /SetVariables(
    Emitter.BottomColorLight:LinearColor=&LerpLinearColors(StartColor=<rgba 0 0.0837 0.2 1>, EndColor=<rgba 0.1 0.0667 0 1>, LerpFactor=$User.ColorProgress)
    Emitter.BottomColorLight02:LinearColor=&LerpLinearColors(StartColor=<rgba 0.0042 0.0066 0.1 1>, EndColor=<rgba 0.1 0.05 0 1>, LerpFactor=$User.ColorProgress)
   ) {isSetParameters=true, module=SetVariables_#4}
  ]
  particleSpawnScript [
   &InitializeParticle(
    Mesh Scale=<0.8 1 1>
    Mesh Scale Mode=::NewEnumerator4
    Position Offset=<-35.1958 0.5 60.7973>
    UsePositionOffset=true
    Write Material Random=true
   )
   &DynamicMaterialParameters(
    Index 0 Param 1=15
    Module.Index0Float0Write=true
    Module.Index0Float1Write=true
    Module.Index0Float2Write=true
    Module.Index0Float3Write=true
   )
   &CameraOffset(Camera Offset Amount=5)
   &InitialMeshOrientation(Mesh Orientation Mode=::NewEnumerator7, Rotation=<0 -0.044 0>, Use Rotation Vector=true)
  ]
  particleUpdateScript [
   &ParticleState(Kill Particles When Lifetime Has Elapsed=false, Let Infinitely Lived Particles Die When Emitter Deactivates=true)
   &ScaleColor(
    Scale RGB=&MakeVectorFromLinearColorRGB(Color=&LerpLinearColors(
      StartColor=&LerpLinearColors(
       StartColor=<rgba 2.2316 4.0664 5 1>
       EndColor=<rgba 2.0589 4.1266 5 1>
       LerpFactor=&FN_FrontEndFloatModifier(Gameplay=0, Front End=1)
      )
      EndColor=<rgba 3 1.5 0 1>
      LerpFactor=$User.ColorProgress
     ))
    ScaleA=true
    ScaleRGB=true
   )
   &FN_Mesh_Attributes(
    Camera Offset=7
    Color=&MakeLinearColorFromVectorAndFloat(
     "Vector (RGB)"=&MakeVectorFromLinearColorRGB(Color=&LerpLinearColors(StartColor=<rgba 1.2417 1.4331 5 1>, EndColor=<rgba 5 3.7472 1.2417 1>, LerpFactor=$User.ColorProgress))
     "Float (Alpha)"=0.2
    )
    Create Camera Offset Attribute=true
    Create Color Attribute=true
    Create Dynamic Parameter Attribute=true
    Create Orientation Attribute=true
    Create Position Attribute=true
    Create Scale Attribute=true
    Index 0 Param 1=10
    Module Output=::NewEnumerator1
    Module.Index0Float0Write=true
    Module.Index0Float1Write=true
    Module.Index0Float2Write=true
    Module.Index0Float3Write=true
    Orientation:Quat4f=<0 -0.0961 0 1>
    Position:NiagaraPosition=<-27.658 0.536 61.5>
    Scale=<0.8 1 1>
   )
   &FN_Mesh_Attributes(
    Camera Offset=3
    Create Camera Offset Attribute=true
    Create Dynamic Parameter Attribute=true
    Create Orientation Attribute=true
    Create Position Attribute=true
    Index 0 Param 1=10
    Module Output=::NewEnumerator1
    Module.Index0Float0Write=true
    Module.Index0Float1Write=true
    Module.Index0Float2Write=true
    Module.Index0Float3Write=true
    Orientation:Quat4f=<0 0 1 1>
    Position:NiagaraPosition=<1 0.24 -1>
   ) {module=FN_Mesh_Attributes002}
  ]
  rendererClasses [&NiagaraMeshRendererProperties]
  renderers [
   {rendererClass=&NiagaraMeshRendererProperties, rendererIndex=0}
   {rendererClass=&NiagaraMeshRendererProperties, rendererIndex=1}
   {rendererClass=&NiagaraMeshRendererProperties, rendererIndex=2}
  ]
 }
 Spike001 {
  emitterUpdateScript [
   &EmitterState(
    Inactive Response=::NewEnumerator1
    Life Cycle Mode=::NewEnumerator1
    Loop Behavior=::NewEnumerator1
    Loop Duration Mode=::NewEnumerator1
    UseLoopDelay=false
   )
   &SpawnBurst_Instantaneous(Use Spawn Probability=false, UseLoopCountLimit=false)
   /SetVariables(
    Emitter.BaseBlueColor:LinearColor=&LerpLinearColors(StartColor=<rgba 0.767 1.234 5 1>, EndColor=<rgba 3 2.2031 1.4063 1>, LerpFactor=$User.ColorProgress)
    Emitter.BaseBlueColor02:LinearColor=&LerpLinearColors(StartColor=<rgba 0 1.8422 3 1>, EndColor=<rgba 3 2.2031 1.4063 1>, LerpFactor=$User.ColorProgress)
   ) {isSetParameters=true, module=SetVariables_#1}
   /SetVariables(
    Emitter.HighLightColor:LinearColor=&LerpLinearColors(StartColor=<rgba 0.1742 0.387 0.5 1>, EndColor=<rgba 1 0.7926 0.5853 1>, LerpFactor=$User.ColorProgress)
    Emitter.HighLightColor02:LinearColor=&LerpLinearColors(StartColor=<rgba 0 1.3561 5 1>, EndColor=<rgba 1 0.7926 0.5853 1>, LerpFactor=$User.ColorProgress)
   ) {isSetParameters=true, module=SetVariables_#2}
   /SetVariables(
    Emitter.BottomColorDark:LinearColor=&LerpLinearColors(StartColor=<rgba 0 0 0.05 1>, EndColor=<rgba 0.05 0.0417 0 1>, LerpFactor=$User.ColorProgress)
    Emitter.BottomColorDark02:LinearColor=&LerpLinearColors(StartColor=<rgba 0 0 0.05 1>, EndColor=<rgba 0.05 0.025 0 1>, LerpFactor=$User.ColorProgress)
   ) {isSetParameters=true, module=SetVariables_#3}
   /SetVariables(
    Emitter.BottomColorLight:LinearColor=&LerpLinearColors(StartColor=<rgba 0 0.0837 0.2 1>, EndColor=<rgba 0.1 0.0667 0 1>, LerpFactor=$User.ColorProgress)
    Emitter.BottomColorLight02:LinearColor=&LerpLinearColors(StartColor=<rgba 0.0042 0.0066 0.1 1>, EndColor=<rgba 0.1 0.05 0 1>, LerpFactor=$User.ColorProgress)
   ) {isSetParameters=true, module=SetVariables_#4}
  ]
  particleSpawnScript [
   &InitializeParticle(Mesh Scale Mode=::NewEnumerator4, UsePositionOffset=true, Write Material Random=true)
   &DynamicMaterialParameters(
    Index 0 Param 1=3
    Module.Index0Float0Write=true
    Module.Index0Float1Write=true
    Module.Index0Float2Write=true
    Module.Index0Float3Write=true
   )
   &CameraOffset(Camera Offset Amount=0.1)
   &InitialMeshOrientation(Mesh Orientation Mode=::NewEnumerator7, Rotation=<0 0 0.25>, Use Rotation Vector=true)
  ]
  particleUpdateScript [
   &ParticleState(Kill Particles When Lifetime Has Elapsed=false, Let Infinitely Lived Particles Die When Emitter Deactivates=true)
   &ScaleColor(
    Scale RGB=&MakeVectorFromLinearColorRGB(Color=&LerpLinearColors(
      StartColor=&LerpLinearColors(
       StartColor=<rgba 2.2316 4.0664 5 1>
       EndColor=<rgba 2.0589 4.1266 5 1>
       LerpFactor=&FN_FrontEndFloatModifier(Gameplay=0, Front End=1)
      )
      EndColor=<rgba 3 1.5 0 1>
      LerpFactor=$User.ColorProgress
     ))
    ScaleA=true
    ScaleRGB=true
   )
   &FN_Mesh_Attributes(
    Camera Offset=0.2
    Color=&MakeLinearColorFromVectorAndFloat(
     "Vector (RGB)"=&MakeVectorFromLinearColorRGB(Color=&LerpLinearColors(StartColor=<rgba 1.2417 1.4331 5 1>, EndColor=<rgba 5 3.7472 1.2417 1>, LerpFactor=$User.ColorProgress))
     "Float (Alpha)"=0.2
    )
    Create Camera Offset Attribute=true
    Create Color Attribute=true
    Create Dynamic Parameter Attribute=true
    Create Orientation Attribute=true
    Create Position Attribute=true
    Create Scale Attribute=true
    Index 0 Param 1=5
    Module Output=::NewEnumerator1
    Module.Index0Float0Write=true
    Module.Index0Float1Write=true
    Module.Index0Float2Write=true
    Module.Index0Float3Write=true
    Orientation:Quat4f=<0 0 1 1>
    Scale=<1 1 1>
   )
   &FN_Mesh_Attributes(
    Create Camera Offset Attribute=true
    Create Dynamic Parameter Attribute=true
    Create Orientation Attribute=true
    Create Position Attribute=true
    Create Scale Attribute=true
    Index 0 Param 1=5
    Module Output=::NewEnumerator1
    Module.Index0Float0Write=true
    Module.Index0Float1Write=true
    Module.Index0Float2Write=true
    Module.Index0Float3Write=true
    Orientation:Quat4f=<0 0 1 1>
    Scale=<1 1 1>
   ) {module=FN_Mesh_Attributes002}
  ]
  rendererClasses [&NiagaraMeshRendererProperties]
  renderers [
   {rendererClass=&NiagaraMeshRendererProperties, rendererIndex=0}
   {rendererClass=&NiagaraMeshRendererProperties, rendererIndex=1}
   {rendererClass=&NiagaraMeshRendererProperties, rendererIndex=2}
  ]
 }
}
summary {
 systemName=NS_Pickaxe_CiderPetalClover_Idle
 emitters [
  {emitterName=Spike, rendererClasses [&NiagaraMeshRendererProperties]}
  {emitterName=Spike001, rendererClasses [&NiagaraMeshRendererProperties]}
 ]
 userVariables [
  User.bIsFrontEnd {type=NiagaraBool}
  User.bIsFrontEndPreview {type=NiagaraBool}
  User.ColorProgress {type=NiagaraFloat, defaultValue=0, description=""}
 ]
}
