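# Niagara swing trail effect for the CiderPetalClover pickaxe, blending from blue to gold/orange through User.ColorProgress.
# Trail emitter builds an anim trail drawn by four ribbon renderers with separate widths and color curves, dimmed in the front end.
# Sparks emitter spawns sprites per distance moved at a socket, pushed by curl noise, gravity and drag, with alpha scaled by time of day.
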NiagaraSystem /BRCosmetics/Weapons/FORT_Melee/Pickaxe_CiderPetalClover/FX/NS_Pickaxe_CiderPetalClover_Trail
dependencies {
 usedDataInterfaces [&NiagaraDataInterfaceColorCurve, &NiagaraDataInterfaceCurve, &NiagaraDataInterfaceSocketReader]
 usedDynamicInputs []
 usedModules [
  &self:SystemSpawnScript.SystemScriptSource.SystemScriptGraph.NiagaraNodeAssignment_0.SetVariables_#1
  &SystemState
 ]
}
emitters {
 Sparks {
  emitterUpdateScript [
   &EmitterState(Life Cycle Mode=::NewEnumerator1, Loop Behavior=::NewEnumerator1, Loop Duration=0.5, UseLoopDelay=false)
   &SpawnPerUnit(Module.Use Max Movement Threshold=false, Module.Use Movement Tolerance=false, Use Spawn Probability=false)
  ]
  particleSpawnScript [
   &InitializeParticle(
    Color Mode=::NewEnumerator0
    Lifetime Max=0.5
    Lifetime Min=0.3
    Lifetime Mode=::NewEnumerator1
    Sprite Size Max=<5 10>
    Sprite Size Min=<5 15>
    Sprite Size Mode=::NewEnumerator2
    UsePositionOffset=false
    Write Material Random=true
   )
   &AddVelocity(Velocity=<0 0 100>)
   &SocketLocation(
    Interpolate Socket Sampling=true
    SamplePosOffset=true
    Socket Index=1
    Socket Reader=$System.Socket Reader
    Socket Sampling Mode=::NewEnumerator3
    VelocitySpeedLimit=false
   )
   &BoxLocation(Box Size=<20 1 1>, MaskSpawnGroup=false, Module.Override Seed=false, Rotate Box=false, Surface Only=false)
   &ApplyInitialForces()
  ]
  particleUpdateScript [
   &ParticleState()
   /SetVariables(Particles.Glow03_Color=&MakeLinearColorFromVectorAndFloat(
     "Vector (RGB)"=&MakeVectorFromLinearColorRGB(Color=&LerpLinearColors(StartColor=<rgba 0.12 0.4133 1 1>, EndColor=<rgba 1 0.7067 0.12 1>, LerpFactor=$User.ColorProgress))
     "Float (Alpha)"=0.1
    )) {isSetParameters=true}
   /SetVariables(Particles.Glow03_Size=&Multiply_Vector2D_ByFloat(
     Float=&FloatFromCurve(
      FloatCurve=DataInterface() {propertyValues {Curve {keys [
          <at 0 0 RCIM_Cubic>
          <at 0.2 1.3 RCIM_Cubic>
          {arriveTangent=-0.1423, interpMode=RCIM_Cubic, leaveTangent=-0.1423, tangentMode=RCTM_User, time=1, value=0}
         ]}}}
      Scale Curve=15
     )
     Vector2D=<1 2>
    )) {isSetParameters=true}
   &ScaleColor(
    Scale Alpha=&FloatFromCurve(
     FloatCurve=DataInterface() {propertyValues {Curve {keys [<at 0.3 1>,<at 0.4 1>,<at 1 1>]}}}
     Scale Curve=&Multiply_Float(A=1, B=&FNTimeOfDayScalar(Night Scale=0.6, Morning Scale=0.8, Evening Scale=0.8))
    )
    Scale RGB=&MakeVectorFromLinearColorRGB(Color=&LerpLinearColors(StartColor=<rgba 0.5 0.7083 3 1>, EndColor=<rgba 3 2 0 1>, LerpFactor=$User.ColorProgress))
    ScaleA=true
    ScaleRGB=true
   )
   &ScaleSpriteSize(Scale Factor=&Vector2DFromFloat(Value=&FloatFromCurve(
      FloatCurve=DataInterface() {propertyValues {Curve {keys [
          <at 0 0 RCIM_Cubic>
          {interpMode=RCIM_Cubic, leaveTangent=-3.0296, tangentMode=RCTM_Break, time=0.2, value=1.2}
          {arriveTangent=-1.8107, interpMode=RCIM_Cubic, leaveTangent=-1.8107, tangentMode=RCTM_User, time=1, value=0}
         ]}}}
      Scale Curve=0.4
     ))) {module=ScaleSpriteSize001}
   &GravityForce(Gravity=<0 0 -200>)
   &CurlNoiseForce(
    Noise Frequency=300
    Noise Strength=&Multiply_Float(A=1000, B=1)
    Pan Noise=true
    Pan Noise Field=<0 0 1>
    Random Seed=10
   )
   &Drag(Drag=3, Use Linear Drag=true, Use Rotational Drag=false)
   &SolveForcesAndVelocity(Clamp Velocity=false, Limit Acceleration=false)
  ]
  rendererClasses [&NiagaraSpriteRendererProperties]
  renderers [
   {rendererClass=&NiagaraSpriteRendererProperties}
   {rendererClass=&NiagaraSpriteRendererProperties, rendererIndex=1}
  ]
 }
 Trail {
  emitterUpdateScript [
   &EmitterState(Life Cycle Mode=::NewEnumerator1, Loop Behavior=::NewEnumerator1, Loop Duration=0.5, UseLoopDelay=false)
   &FN_SpawnAnimTrail()
  ]
  particleSpawnScript [
   &FN_InitializeAnimTrailBoneIndecies()
   &FN_InitializeAnimTrail(AnimTrailWidth=2.2, Lifetime=0.3)
  ]
  particleUpdateScript [
   &FN_UpdateAnimTrail()
   /SetVariables(
    Particles.RibbonWidth Crystal:NiagaraFloat=&Multiply_Float(A=$Particles.RibbonWidth, B=1.5)
    Particles.RibbonWidth Glow:NiagaraFloat=&Multiply_Float(A=$Particles.RibbonWidth, B=1.5)
    Particles.RibbonWidth Hightlight:NiagaraFloat=&Multiply_Float(A=$Particles.RibbonWidth, B=1)
    Particles.RibbonWidth Light:NiagaraFloat=&Multiply_Float(A=$Particles.RibbonWidth, B=1.4)
   ) {isSetParameters=true}
   &ScaleColor(
    Scale Alpha=&FloatFromCurve(
     FloatCurve=DataInterface() {propertyValues {Curve {keys [
         {arriveTangent=-2.1148, interpMode=RCIM_Cubic, leaveTangent=-2.1148, tangentMode=RCTM_User, time=0, value=0.8}
         {arriveTangent=-0.95, interpMode=RCIM_Cubic, leaveTangent=-0.95, tangentMode=RCTM_User, time=0.3, value=0.25}
         <at 0.7 0 RCIM_Cubic>
        ]}}}
     Scale Curve=&Multiply_Float(
      A=&FloatFromCurve(FloatCurve=DataInterface() {propertyValues {Curve {keys [<at 0.2 1>,<at 0.7 0>]}}}, CurveIndex=$Emitter.Age)
      B=&Lerp_Float(A=6, B=3, Alpha=$User.ColorProgress)
     )
    )
    Scale RGB=&Multiply_VectorByFloat(
     Vector=&MakeVectorFromLinearColorRGB(Color=&LerpLinearColors(
       StartColor=&ColorFromCurve(ColorCurve=DataInterface() {propertyValues {
          AlphaCurve {keys []}
          BlueCurve {keys [{arriveTangent=-0.034, interpMode=RCIM_Cubic, time=0.0778, value=1}, <at 0.3059 0.5>]}
          GreenCurve {keys [{arriveTangent=-0.0023, interpMode=RCIM_Cubic, time=0.0778, value=0.25}, <at 0.3059 0.0612>]}
          RedCurve {keys [{arriveTangent=-0.0931, interpMode=RCIM_Cubic, time=0.0778, value=0.1}, <at 0.3059 0.0214>]}
         }})
       EndColor=&ColorFromCurve(ColorCurve=DataInterface() {propertyValues {
          AlphaCurve {keys []}
          BlueCurve {keys [{arriveTangent=-0.034, interpMode=RCIM_Cubic, time=0.0778, value=0.0986}, <at 0.4906 0.0215>]}
          GreenCurve {keys [{arriveTangent=-0.0023, interpMode=RCIM_Cubic, time=0.0778, value=0.849}, <at 0.4906 0.2291>]}
          RedCurve {keys [{arriveTangent=-0.0931, interpMode=RCIM_Cubic, time=0.0778, value=1}, <at 0.4906 0.5>]}
         }})
       LerpFactor=$User.ColorProgress
      ))
     Float=&Multiply_Float(
      A=&FN_FrontEndFloatModifier(Gameplay=1, Front End=0.5)
      B=&Lerp_Float(A=1, B=0.5, Alpha=$User.ColorProgress)
     )
    )
    ScaleA=true
    ScaleRGB=true
   )
   &FN_Sprite_Attributes(
    Color=&MakeLinearColorFromVectorAndFloat(
     "Vector (RGB)"=&Multiply_VectorByFloat(
      Vector=&MakeVectorFromLinearColorRGB(Color=&LerpLinearColors(
        StartColor=&ColorFromCurve(ColorCurve=DataInterface() {propertyValues {
           AlphaCurve {keys []}
           BlueCurve {keys [{arriveTangent=-0.034, interpMode=RCIM_Cubic, time=0.2143, value=4}, <at 0.7864 0.083>]}
           GreenCurve {keys [{arriveTangent=-0.0023, interpMode=RCIM_Cubic, time=0.2143, value=1.3333}, <at 0.7864 0.0069>]}
           RedCurve {keys [{arriveTangent=-0.0931, interpMode=RCIM_Cubic, time=0.2143, value=0}, <at 0.7864 0>]}
          }})
        EndColor=&ColorFromCurve(ColorCurve=DataInterface() {propertyValues {
           AlphaCurve {keys []}
           BlueCurve {keys [{arriveTangent=-0.034, interpMode=RCIM_Cubic, time=0.2143, value=0}, <at 0.7864 0>]}
           GreenCurve {keys [{arriveTangent=-0.0023, interpMode=RCIM_Cubic, time=0.2143, value=2.6667}, <at 0.7864 0.0553>]}
           RedCurve {keys [{arriveTangent=-0.0931, interpMode=RCIM_Cubic, time=0.2143, value=4}, <at 0.7864 0.083>]}
          }})
        LerpFactor=$User.ColorProgress
       ))
      Float=&Multiply_Float(
       A=&FN_FrontEndFloatModifier(Gameplay=0.5, Front End=0.5)
       B=&Lerp_Float(A=1, B=0.3, Alpha=$User.ColorProgress)
      )
     )
     "Float (Alpha)"=&FloatFromCurve(FloatCurve=DataInterface() {propertyValues {Curve {keys [<at 0 1>,<at 0.6 1>,<at 1 0>]}}}, Scale Curve=0.06)
    )
    Create Color Attribute=true
    Module Output:ENiagara_AttributeSamplingApplyOutput=::NewEnumerator1
   )
   &FN_Sprite_Attributes(
    Color=&MakeLinearColorFromVectorAndFloat(
     "Vector (RGB)"=&Multiply_VectorByFloat(
      Vector=&MakeVectorFromLinearColorRGB(Color=&LerpLinearColors(
        StartColor=&ColorFromCurve(ColorCurve=DataInterface() {propertyValues {
           AlphaCurve {keys []}
           BlueCurve {keys [
             {arriveTangent=-11.7847, interpMode=RCIM_Cubic, leaveTangent=-11.7847, tangentMode=RCTM_User, time=0, value=20}
             <at 0.2495 3>
             <at 0.5887 0.4>
            ]}
           GreenCurve {keys [
             {arriveTangent=-4.0036, interpMode=RCIM_Cubic, leaveTangent=-4.0036, tangentMode=RCTM_User, time=0, value=8}
             <at 0.2495 1.2271>
             <at 0.5887 0.1547>
            ]}
           RedCurve {keys [
             {arriveTangent=-34.1305, interpMode=RCIM_Cubic, leaveTangent=-34.1305, tangentMode=RCTM_User, time=0, value=2}
             <at 0.2495 0.3406>
             <at 0.5887 0.0321>
            ]}
          }})
        EndColor=&ColorFromCurve(ColorCurve=DataInterface() {propertyValues {
           AlphaCurve {keys []}
           BlueCurve {keys [
             {arriveTangent=-11.7847, interpMode=RCIM_Cubic, leaveTangent=-11.7847, tangentMode=RCTM_User, time=0.3011, value=0.4948}
             <at 0.3902 0>
             <at 0.6061 0.0234>
            ]}
           GreenCurve {keys [
             {arriveTangent=-4.0036, interpMode=RCIM_Cubic, leaveTangent=-4.0036, tangentMode=RCTM_User, time=0.3011, value=2.7171}
             <at 0.3902 1.0816>
             <at 0.6061 0.1411>
            ]}
           RedCurve {keys [
             {arriveTangent=-34.1305, interpMode=RCIM_Cubic, leaveTangent=-34.1305, tangentMode=RCTM_User, time=0.3011, value=5}
             <at 0.3902 3>
             <at 0.6061 0.2>
            ]}
          }})
        LerpFactor=$User.ColorProgress
       ))
      Float=&FN_FrontEndFloatModifier(Gameplay=1, Front End=0.5)
     )
     "Float (Alpha)"=&FloatFromCurve(
      FloatCurve=DataInterface() {propertyValues {Curve {keys [
          <at 0 1>
          {arriveTangent=-1.716, interpMode=RCIM_Cubic, leaveTangent=-1.716, tangentMode=RCTM_User, time=0.6, value=1}
          {arriveTangent=0.05, interpMode=RCIM_Cubic, leaveTangent=0.05, tangentMode=RCTM_User, time=1, value=0}
         ]}}}
      Scale Curve=0.5
     )
    )
    Create Color Attribute=true
    Module Output:ENiagara_AttributeSamplingApplyOutput=::NewEnumerator1
   ) {module=FN_Sprite_Attributes001}
   &FN_Sprite_Attributes(
    Color=&MakeLinearColorFromVectorAndFloat(
     "Vector (RGB)"=&Multiply_VectorByFloat(
      Vector=&MakeVectorFromLinearColorRGB(Color=&LerpLinearColors(
        StartColor=&ColorFromCurve(ColorCurve=DataInterface() {propertyValues {
           AlphaCurve {keys []}
           BlueCurve {keys [<at 0 10>,<at 0.2401 2>,<at 0.7453 0.5>]}
           GreenCurve {keys [<at 0 1.6667>,<at 0.2401 0.6847>,<at 0.7453 0.0833>]}
           RedCurve {keys [<at 0 0>,<at 0.2401 0.0271>,<at 0.7453 0>]}
          }})
        EndColor=&Multiply_LinearColorByFloat(
         Linear Color=&ColorFromCurve(ColorCurve=DataInterface() {propertyValues {
            AlphaCurve {keys []}
            BlueCurve {keys [{arriveTangent=-46.1537, interpMode=RCIM_Cubic, time=0, value=0.04}, <at 0.2448 0>, <at 0.5719 0>]}
            GreenCurve {keys [{arriveTangent=-2.1203, interpMode=RCIM_Cubic, time=0, value=1.3467}, <at 0.2448 0.6774>, <at 0.5719 0.1895>]}
            RedCurve {keys [{arriveTangent=-2.3077, interpMode=RCIM_Cubic, time=0, value=2}, <at 0.2448 1>, <at 0.5719 0.5>]}
           }})
         Float=0.7
        )
        LerpFactor=$User.ColorProgress
       ))
      Float=&FN_FrontEndFloatModifier(Gameplay=1, Front End=0.5)
     )
     "Float (Alpha)"=&FloatFromCurve(
      FloatCurve=DataInterface() {propertyValues {Curve {keys [<at 0.3 1>,<at 0.6 0>]}}}
      Scale Curve=&Lerp_Float(A=1.5, B=1.5, Alpha=$User.ColorProgress)
     )
    )
    Create Color Attribute=true
    Module Output:ENiagara_AttributeSamplingApplyOutput=::NewEnumerator1
   ) {module=FN_Sprite_Attributes003}
  ]
  rendererClasses [&NiagaraRibbonRendererProperties]
  renderers [
   {rendererClass=&NiagaraRibbonRendererProperties}
   {rendererClass=&NiagaraRibbonRendererProperties, rendererIndex=1}
   {rendererClass=&NiagaraRibbonRendererProperties, rendererIndex=2}
   {rendererClass=&NiagaraRibbonRendererProperties, rendererIndex=3}
  ]
 }
}
summary {
 systemName=NS_Pickaxe_CiderPetalClover_Trail
 emitters [
  {emitterName=Trail, rendererClasses [&NiagaraRibbonRendererProperties]}
  {emitterName=Sparks, rendererClasses [&NiagaraSpriteRendererProperties]}
 ]
 userVariables [
  User.bIsFrontEnd {type=NiagaraBool}
  User.bIsFrontEndPreview {type=NiagaraBool}
  User.ColorProgress {type=NiagaraFloat, defaultValue=0, description=""}
  User.RibbonIDOffset {type=NiagaraInt32}
 ]
}
